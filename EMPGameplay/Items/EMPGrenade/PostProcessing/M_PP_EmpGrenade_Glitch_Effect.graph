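# Post-process material for the EMP grenade that turns the screen into a blue-tinted electrical glitch.
# A time-driven interval mask gates noise-based UV distortion, and three offset scene samples create RGB chromatic aberration.
# An overlay tint, blinking screen-blended vignette and scrolling sine scanlines are summed, then blended over the original scene by OverallEffectIntensity.
# Named reroutes link the stages, and a few texture samples are left unconnected.
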Material /EMPGameplay/Items/EMPGrenade/PostProcessing/M_PP_EmpGrenade_Glitch_Effect
graph {
 properties {materialDomain=MD_PostProcess, pixelDepthOffsetMode=PDOM_Legacy, usedShadingModels=Default Lit}
 nodes [
  n1 = Multiply(A=NamedRerouteUsage()) {constB=0.88}
  n2 = ScalarParameter() {defaultValue=2, parameterName=Glitch Interval Speed}
  n3 = Sine(None=Multiply(A=@n1, B=@n2))
  Abs(None=@n3)
  n4 = NamedRerouteUsage()
  n5 = VectorParameter() {parameterName=Main Color, defaultValue=<rgba 0.0976 0.6795 1 1>}
  n6 = LinearInterpolate(
   A=ComponentMask(None=@n4) {b=true, g=true, r=true}
   B=MaterialFunctionCall(Base=@n4, Blend=@n5.RGB) {materialFunction=&Blend_Overlay}.Result
   Alpha=ScalarParameter() {defaultValue=0.5, parameterName=Blue Tint Amount}
  )
  n7 = Distance(A=NamedRerouteUsage(), B=Constant() {r=0.5})
  n8 = ScalarParameter() {defaultValue=3, parameterName=Vignette Smoothness}
  n9 = ScalarParameter() {defaultValue=1, parameterName=VignetteIntensity}
  n10 = NamedRerouteUsage()
  n11 = ScalarParameter() {defaultValue=2, parameterName=Vignette Blink Speed}
  n12 = Abs(None=Sine(None=Multiply(A=@n10, B=@n11) {constB=2}))
  n13 = Multiply(A=@n12, B=Multiply(A=Constant() {r=0.15}, B=ScalarParameter() {defaultValue=1, parameterName=Vignette Blink Intensity})) {constB=0.15}
  n14 = LinearInterpolate(
   B=MaterialFunctionCall!(Blend=@n4, Base=@n5.RGB) {materialFunction=/Engine/Functions/Engine_MaterialFunctions03/Blends/Blend_Screen}.Result
   Alpha=Multiply(A=Power(Base=@n7, Exp=@n8), B=Add(A=@n9, B=@n13)) {constB=2}
  )
  n15 = NamedRerouteUsage()
  n16 = Multiply(A=Time() {bOverride_Period=true, period=0.1}, B=ScalarParameter() {defaultValue=0.1, parameterName=Lines Speed}) {constB=0.1}
  n17 = ComponentMask(None=Multiply(A=Subtract(A=@n15, B=@n16), B=ScalarParameter() {defaultValue=50, parameterName=Lines Amount})) {g=true}
  n18 = Multiply(
   A=Multiply(A=VectorParameter() {parameterName=Main Color, defaultValue=<rgba 0.0976 0.6795 1 1>}.RGB) {constB=0.3}
   B=Saturate(None=Sine(None=@n17))
  )
  n19 = ScalarParameter() {defaultValue=0.5, parameterName=Lines Opacity}
  LinearInterpolate(
   A=ComponentMask(None=SceneTexture() {sceneTextureId=PPI_PostProcessInput0}.Color) {b=true, g=true, r=true}
   B=Add(A=Add(A=@n6, B=@n14), B=Multiply(A=@n18, B=@n19) {constB=0})
   Alpha=Saturate(None=ScalarParameter() {defaultValue=1, parameterName=OverallEffectIntensity, sliderMax=1})
  ) {constAlpha=1}
  NamedRerouteDeclaration(None=Time() {period=10}) {name=Time Var, nodeColor=<rgba 0 0.9766 1 1>}
  n20 = OneMinus(None=@n3)
  n21 = Frac(None=Multiply(A=Multiply(A=@n1) {constB=0.77}, B=@n2))
  NamedRerouteDeclaration(None=Round(None=Multiply(A=@n20, B=@n21)) {desc=Interval Mask}) {name=Frequency Mask, nodeColor=<rgba 1 0.4922 0 1>}
  NamedRerouteDeclaration(None=TextureCoordinate()) {name=UV, nodeColor=<rgba 1 0.0932 0.1231 1>}
  n22 = Multiply(A=NamedRerouteUsage(), B=Multiply(A=Constant() {r=0.01}, B=ScalarParameter() {defaultValue=1, parameterName=ChromaticIntensity})) {constB=0.01}
  n23 = NamedRerouteUsage()
  n24 = SceneTexture(UVs=Add(A=@n23, B=@n22)) {sceneTextureId=PPI_PostProcessInput0}
  n25 = SceneTexture(UVs=@n23) {sceneTextureId=PPI_PostProcessInput0}
  n26 = SceneTexture(UVs=Subtract(A=@n23, B=@n22)) {sceneTextureId=PPI_PostProcessInput0}
  NamedRerouteDeclaration(None=MaterialFunctionCall(X=ComponentMask(None=@n24.Color) {r=true}, Y=ComponentMask(None=@n25.Color) {g=true}, Z=ComponentMask(None=@n26.Color) {b=true}) {materialFunction=&MakeFloat3}.Result) {name=ChromaticTexture, nodeColor=<rgba 0 0.9766 1 1>}
  n27 = NamedRerouteUsage()
  n28 = Sine(None=ComponentMask(None=Panner(Coordinate=@n27) {speedX=3}) {r=true})
  n29 = Panner(Coordinate=@n27) {speedX=7}
  n30 = Multiply(A=Sine(None=ComponentMask(None=@n29) {r=true}) {period=10}, B=ScalarParameter() {defaultValue=0.5, parameterName=Frequency})
  n31 = NamedRerouteUsage()
  n32 = NamedRerouteUsage()
  n33 = Multiply(A=@n32) {constB=0.9}
  n34 = TextureSample(UVs=Multiply(A=Subtract(A=@n31, B=@n33), B=ScalarParameter() {defaultValue=1, parameterName=Glitch Size})) {
   samplerType=SAMPLERTYPE_Masks
   texture=/Game/Characters/Player/Female/Medium/Bodies/F_MED_BlueBadass/Skins/NeonLines/F_MED_BlueBadass_NeonLines_-_GlitchNoise_v1
  }
  n35 = Multiply(A=Add(A=Ceil(None=@n28), B=Ceil(None=@n30)), B=LinearInterpolate(B=@n34.R, Alpha=NamedRerouteUsage()))
  n36 = Multiply(
   A=Subtract(A=Multiply(A=@n35) {constB=2})
   B=Multiply(A=Constant() {r=0.03}, B=ScalarParameter() {defaultValue=0.3, parameterName=Glitch Intensity})
  )
  NamedRerouteDeclaration(None=Add(A=NamedRerouteUsage(), B=Saturate(None=@n36))) {name=Glitched UV, nodeColor=<rgba 1 0.6094 0 1>}
  Round(None=Multiply(A=@n32) {constB=3})
  Round(None=@n12)
  TextureSample() {samplerType=SAMPLERTYPE_LinearGrayscale, texture=/Game/Global/Textures/Default/Noise/glitch_noise}
  TextureSample() {texture=/Game/Accessories/FORT_Backpacks/Backpack_M_MED_FN_Birthday_5/Textures/TechArt/T_M_MED_FN_Birthday_5_Backpack_Glitch_Confetti_01}
  TextureSample() {texture=/Game/Accessories/FORT_Backpacks/Backpack_M_MED_FN_Birthday_5/Textures/TechArt/T_M_MED_FN_Birthday_5_Backpack_Glitch_Confetti_03}
 ]
}
